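# Physics asset for the Sailor Squad Rebel female medium face accessory mesh, set up for two dynamic ponytails.
# Each ponytail is a mirrored capsule chain from a small base body on the head through six segments, with limited swing and tight twist between segments.
# Capsules on pelvis, spine, neck, head and arms form a body proxy, joined by constraints with free angular motion, as are the ponytail bases to the head.
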PhysicsAsset /BRCosmetics/Characters/Player/Female/Medium/Bodies/F_MED_Sailor_Squad_Rebel/Meshes/Parts/F_MED_Sailor_Squad_Rebel_FaceAcc_Physics
bodies {
 dyn_ponytail_a_base_l [{shapeType=Capsule, length=0.505, radius=0.505, rotation=<rot 0 -90 0>}]
 dyn_ponytail_a_base_r [{shapeType=Capsule, length=0.505, radius=0.505, rotation=<rot 0 -90 180>}]
 dyn_ponytail_a_l [{shapeType=Capsule, length=4.25, radius=2.5, center=<4.4522 0 0>, rotation=<rot 90 0 0>}]
 dyn_ponytail_a_r [{shapeType=Capsule, length=4.25, radius=2.5, center=<-4.4522 0 0>, rotation=<rot 90 0 180>}]
 dyn_ponytail_b_l [{shapeType=Capsule, length=4, radius=3, center=<5.152 0 0>, rotation=<rot 90 0 0>}]
 dyn_ponytail_b_r [{shapeType=Capsule, length=4, radius=3, center=<-5.152 0 0>, rotation=<rot 90 0 180>}]
 dyn_ponytail_c_l [{shapeType=Capsule, length=3, radius=3.25, center=<2.8165 0 0>, rotation=<rot 90 0 0>}]
 dyn_ponytail_c_r [{shapeType=Capsule, length=3, radius=3.25, center=<-2.8165 0 0>, rotation=<rot 90 0 180>}]
 dyn_ponytail_d_l [{shapeType=Capsule, length=2.25, radius=4, center=<4.1487 0 0.0001>, rotation=<rot 90 0 0>}]
 dyn_ponytail_d_r [{shapeType=Capsule, length=2.25, radius=4, center=<-4.1487 0 -0.0001>, rotation=<rot 90 0 180>}]
 dyn_ponytail_e_l [{shapeType=Capsule, length=0.5, radius=5.75, center=<3.6729 0 -0.0001>, rotation=<rot 90 0 0>}]
 dyn_ponytail_e_r [{shapeType=Capsule, length=0.5, radius=5.75, center=<-3.6729 0 0.0001>, rotation=<rot 90 0 180>}]
 dyn_ponytail_f_l [{shapeType=Capsule, length=3.5, radius=5.25, center=<6.405 0 0>, rotation=<rot 90 0 0>}]
 dyn_ponytail_f_r [{shapeType=Capsule, length=3.5, radius=5.25, center=<-6.405 0 0>, rotation=<rot 90 0 180>}]
 head [{shapeType=Capsule, length=6, radius=7, center=<3.51 2.2324 0>, rotation=<rot 90 0 0>}]
 lowerarm_l [{shapeType=Capsule, length=23.5, radius=4.75, center=<12.7106 0 0>, rotation=<rot 90 0 0>}]
 lowerarm_r [{shapeType=Capsule, length=23.5, radius=4.75, center=<-12.7106 0 0>, rotation=<rot 90 0 180>}]
 neck_01 [{shapeType=Capsule, length=4.5, radius=4.25, center=<1.5642 0 0>, rotation=<rot 90 0 0>}]
 pelvis [{shapeType=Capsule, length=9.25, radius=12.75, center=<-0.8809 1.5936 0>}]
 spine_02 [{shapeType=Capsule, length=9.75, radius=8.25, center=<-0.8665 2.1785 0.0001>}]
 spine_03 [{shapeType=Capsule, length=9.5, radius=9.25, center=<4.9633 3.5337 0>}]
 spine_05 [
  {shapeType=Capsule, length=72.25, radius=8.75, center=<1.2073 1.1263 0.0002>}
  {shapeType=Capsule, length=68.5, radius=12.5, shapeName=spine_05_capsule, center=<17.1096 10.7064 0>}
 ]
 upperarm_l [{shapeType=Capsule, length=21.25, radius=5, center=<11.9562 0 0>, rotation=<rot 90 0 0>}]
 upperarm_r [{shapeType=Capsule, length=21.25, radius=5, center=<-11.9562 0 0>, rotation=<rot 90 0 180>}]
}
constraints [
 {
  bone1Name=neck_01
  bone2Name=spine_05
  swing1LimitDegrees=45
  swing1Motion=Free
  swing2LimitDegrees=45
  swing2Motion=Free
  twistLimitDegrees=45
  twistMotion=Free
 }
 {
  bone1Name=upperarm_l
  bone2Name=spine_05
  swing1LimitDegrees=45
  swing1Motion=Free
  swing2LimitDegrees=45
  swing2Motion=Free
  twistLimitDegrees=45
  twistMotion=Free
 }
 {
  bone1Name=upperarm_r
  bone2Name=spine_05
  swing1LimitDegrees=45
  swing1Motion=Free
  swing2LimitDegrees=45
  swing2Motion=Free
  twistLimitDegrees=45
  twistMotion=Free
 }
 {
  bone1Name=dyn_ponytail_a_base_l
  bone2Name=head
  swing1LimitDegrees=45
  swing1Motion=Free
  swing2LimitDegrees=45
  swing2Motion=Free
  twistLimitDegrees=45
  twistMotion=Free
 }
 {bone1Name=dyn_ponytail_a_l, bone2Name=dyn_ponytail_a_base_l, swing1LimitDegrees=32, swing2LimitDegrees=15, twistLimitDegrees=5}
 {bone1Name=dyn_ponytail_b_l, bone2Name=dyn_ponytail_a_l, swing1LimitDegrees=50, swing2LimitDegrees=15, twistLimitDegrees=5}
 {bone1Name=dyn_ponytail_c_l, bone2Name=dyn_ponytail_b_l, swing1LimitDegrees=60, swing2LimitDegrees=5, twistLimitDegrees=5}
 {bone1Name=dyn_ponytail_d_l, bone2Name=dyn_ponytail_c_l, swing1LimitDegrees=60, swing2LimitDegrees=5, twistLimitDegrees=5}
 {bone1Name=dyn_ponytail_e_l, bone2Name=dyn_ponytail_d_l, swing1LimitDegrees=60, swing2LimitDegrees=5, twistLimitDegrees=5}
 {bone1Name=dyn_ponytail_f_l, bone2Name=dyn_ponytail_e_l, swing1LimitDegrees=60, swing2LimitDegrees=5, twistLimitDegrees=5}
 {
  bone1Name=spine_05
  bone2Name=spine_03
  swing1LimitDegrees=45
  swing1Motion=Free
  swing2LimitDegrees=45
  swing2Motion=Free
  twistLimitDegrees=45
  twistMotion=Free
 }
 {
  bone1Name=spine_03
  bone2Name=spine_02
  swing1LimitDegrees=45
  swing1Motion=Free
  swing2LimitDegrees=45
  swing2Motion=Free
  twistLimitDegrees=45
  twistMotion=Free
 }
 {
  bone1Name=spine_02
  bone2Name=pelvis
  swing1LimitDegrees=45
  swing1Motion=Free
  swing2LimitDegrees=45
  swing2Motion=Free
  twistLimitDegrees=45
  twistMotion=Free
 }
 {
  bone1Name=dyn_ponytail_a_base_r
  bone2Name=head
  swing1LimitDegrees=45
  swing1Motion=Free
  swing2LimitDegrees=45
  swing2Motion=Free
  twistLimitDegrees=45
  twistMotion=Free
 }
 {bone1Name=dyn_ponytail_a_r, bone2Name=dyn_ponytail_a_base_r, swing1LimitDegrees=32, swing2LimitDegrees=15, twistLimitDegrees=5}
 {bone1Name=dyn_ponytail_b_r, bone2Name=dyn_ponytail_a_r, swing1LimitDegrees=50, swing2LimitDegrees=15, twistLimitDegrees=5}
 {bone1Name=dyn_ponytail_c_r, bone2Name=dyn_ponytail_b_r, swing1LimitDegrees=60, swing2LimitDegrees=5, twistLimitDegrees=5}
 {bone1Name=dyn_ponytail_d_r, bone2Name=dyn_ponytail_c_r, swing1LimitDegrees=60, swing2LimitDegrees=5, twistLimitDegrees=5}
 {bone1Name=dyn_ponytail_e_r, bone2Name=dyn_ponytail_d_r, swing1LimitDegrees=60, swing2LimitDegrees=5, twistLimitDegrees=5}
 {bone1Name=dyn_ponytail_f_r, bone2Name=dyn_ponytail_e_r, swing1LimitDegrees=60, swing2LimitDegrees=5, twistLimitDegrees=5}
 {
  bone1Name=lowerarm_l
  bone2Name=upperarm_l
  swing1LimitDegrees=45
  swing1Motion=Free
  swing2LimitDegrees=45
  swing2Motion=Free
  twistLimitDegrees=45
  twistMotion=Free
 }
 {
  bone1Name=lowerarm_r
  bone2Name=upperarm_r
  swing1LimitDegrees=45
  swing1Motion=Free
  swing2LimitDegrees=45
  swing2Motion=Free
  twistLimitDegrees=45
  twistMotion=Free
 }
 {
  bone1Name=head
  bone2Name=neck_01
  swing1LimitDegrees=45
  swing1Motion=Free
  swing2LimitDegrees=45
  swing2Motion=Free
  twistLimitDegrees=45
  twistMotion=Free
 }
]
