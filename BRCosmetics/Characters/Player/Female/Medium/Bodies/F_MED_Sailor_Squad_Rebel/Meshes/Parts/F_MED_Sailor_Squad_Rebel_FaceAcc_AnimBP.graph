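# Animation Blueprint for the Sailor Squad Rebel face accessory that animates twin ponytails and front hair on a pose copied from the body.
# The dynamic path adjusts ponytails for head angle, jumping and skydiving around RigidBody simulation, with AnimDynamics on front hair.
# When rigid body is off, LOD is 1+ or back bling is equipped, weighted bone constraints and fixed offsets replace ponytail simulation, keeping front-hair AnimDynamics.
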AnimBlueprint /BRCosmetics/Characters/Player/Female/Medium/Bodies/F_MED_Sailor_Squad_Rebel/Meshes/Parts/F_MED_Sailor_Squad_Rebel_FaceAcc_AnimBP
target_skeleton=&/F_MED_Sailor_Squad_Rebel_FaceAcc_Skeleton
graphs [
 AnimGraph {nodes [
   n1 = Use cached pose 'StartPose'()
   n2 = Use cached pose 'DynamicPose'()
   n3 = Get bRigidBodyEnabled()
   Output Pose(Result=Blend Poses by bool(
     BlendPose_0=Use cached pose 'NoDynamicPose'()
     BlendPose_1=Blend Poses by bool(BlendPose_0=@n2, BlendPose_1=@n1, bActiveValue=@n3) {blendType=Linear}
     bActiveValue=Get IsRigidBodyDisabled() -> (IsRigidBodyDisabled:bool)
    ) {blendType=Linear}) {layerGroup=None}
   n4 = "Constraint - dyn_ponytail_a_l (PPPP)"!(
    ComponentPose=Local To Component(LocalPose=Use cached pose 'StartPose'())
    Alpha=1
    ConstraintWeights_0=0.0547
    ConstraintWeights_1:float=0.0119
    ConstraintWeights_2:float=0.9105
    ConstraintWeights_3:float=0.0228
   ) -> (Pose) {
    alphaBoolBlend {blendOption=Linear}
    boneToModify {boneName=dyn_ponytail_a_l}
    constraintSetup [{targetBone {boneName=neck_02}}, {targetBone {boneName=neck_01}}, {targetBone {boneName=head}}, {targetBone {boneName=spine_05}}]
    constraintWeights [1,1,1,1]
   }
   n5 = "Constraint - dyn_ponytail_b_l (PPPPPP)"!(
    ComponentPose=@n4
    Alpha=1
    ConstraintWeights_0=0.034
    ConstraintWeights_1:float=0.0635
    ConstraintWeights_2:float=0.0152
    ConstraintWeights_3:float=0.504
    ConstraintWeights_4:float=0.2234
    ConstraintWeights_5:float=0.1599
   ) -> (Pose) {
    alphaBoolBlend {blendOption=Linear}
    boneToModify {boneName=dyn_ponytail_b_l}
    constraintSetup [
     {targetBone {boneName=head}}
     {targetBone {boneName=clavicle_l}}
     {targetBone {boneName=spine_03}}
     {targetBone {boneName=spine_05}}
     {targetBone {boneName=neck_02}}
     {targetBone {boneName=neck_01}}
    ]
    constraintWeights [1,1,1,1,1,1]
   }
   n6 = "Constraint - dyn_ponytail_c_l (PPP)"!(ComponentPose=@n5, Alpha=1, ConstraintWeights_0=0.0436, ConstraintWeights_1:float=0.3414, ConstraintWeights_2:float=0.6149) -> (Pose) {
    alphaBoolBlend {blendOption=Linear}
    boneToModify {boneName=dyn_ponytail_c_l}
    constraintSetup [{targetBone {boneName=spine_04}}, {targetBone {boneName=spine_05}}, {targetBone {boneName=clavicle_l}}]
    constraintWeights [1,1,1]
   }
   n7 = "Constraint - dyn_ponytail_d_l (PPPPPP)"!(
    ComponentPose=@n6
    Alpha=1
    ConstraintWeights_0=0.109
    ConstraintWeights_1:float=0.0409
    ConstraintWeights_2:float=0.0346
    ConstraintWeights_3:float=0.3289
    ConstraintWeights_4:float=0.4743
    ConstraintWeights_5:float=0.0123
   ) -> (Pose) {
    alphaBoolBlend {blendOption=Linear}
    boneToModify {boneName=dyn_ponytail_d_l}
    constraintSetup [
     {targetBone {boneName=clavicle_l}}
     {targetBone {boneName=spine_02}}
     {targetBone {boneName=spine_03}}
     {targetBone {boneName=spine_04}}
     {targetBone {boneName=spine_05}}
     {targetBone {boneName=pelvis}}
    ]
    constraintWeights [1,1,1,1,1,1]
   }
   n8 = "Constraint - dyn_ponytail_e_l (PPPPPP)"!(
    ComponentPose=@n7
    Alpha=1
    ConstraintWeights_0=0.0134
    ConstraintWeights_1:float=0.0748
    ConstraintWeights_2:float=0.066
    ConstraintWeights_3:float=0.5
    ConstraintWeights_4:float=0.3264
    ConstraintWeights_5:float=0.0195
   ) -> (Pose) {
    alphaBoolBlend {blendOption=Linear}
    boneToModify {boneName=dyn_ponytail_e_l}
    constraintSetup [
     {targetBone {boneName=spine_01}}
     {targetBone {boneName=spine_02}}
     {targetBone {boneName=spine_03}}
     {targetBone {boneName=spine_04}}
     {targetBone {boneName=spine_05}}
     {targetBone {boneName=pelvis}}
    ]
    constraintWeights [1,1,1,1,1,1]
   }
   n9 = "Constraint - dyn_ponytail_f_l (PPPPPP)"!(
    ComponentPose=@n8
    Alpha=1
    ConstraintWeights_0=0.0418
    ConstraintWeights_1:float=0.1145
    ConstraintWeights_2:float=0.0756
    ConstraintWeights_3:float=0.3322
    ConstraintWeights_4:float=0.1273
    ConstraintWeights_5:float=0.3086
   ) -> (Pose) {
    alphaBoolBlend {blendOption=Linear}
    boneToModify {boneName=dyn_ponytail_f_l}
    constraintSetup [
     {targetBone {boneName=spine_01}}
     {targetBone {boneName=spine_02}}
     {targetBone {boneName=spine_03}}
     {targetBone {boneName=spine_04}}
     {targetBone {boneName=spine_05}}
     {targetBone {boneName=pelvis}}
    ]
    constraintWeights [1,1,1,1,1,1]
   }
   n10 = "Constraint - dyn_ponytail_a_r (PPPP)"!(
    ComponentPose=@n9
    Alpha=1
    ConstraintWeights_0=0.0847
    ConstraintWeights_1:float=0.0208
    ConstraintWeights_2:float=0.8633
    ConstraintWeights_3:float=0.0312
   ) -> (Pose) {
    alphaBoolBlend {blendOption=Linear}
    boneToModify {boneName=dyn_ponytail_a_r}
    constraintSetup [{targetBone {boneName=neck_02}}, {targetBone {boneName=neck_01}}, {targetBone {boneName=head}}, {targetBone {boneName=spine_05}}]
    constraintWeights [1,1,1,1]
   }
   n11 = "Constraint - dyn_ponytail_b_r (PPPPPP)"!(
    ComponentPose=@n10
    Alpha=1
    ConstraintWeights_0=0.0343
    ConstraintWeights_1:float=0.0162
    ConstraintWeights_2:float=0.5044
    ConstraintWeights_3:float=0.0743
    ConstraintWeights_4:float=0.2158
    ConstraintWeights_5:float=0.155
   ) -> (Pose) {
    alphaBoolBlend {blendOption=Linear}
    boneToModify {boneName=dyn_ponytail_b_r}
    constraintSetup [
     {targetBone {boneName=head}}
     {targetBone {boneName=spine_03}}
     {targetBone {boneName=spine_05}}
     {targetBone {boneName=clavicle_r}}
     {targetBone {boneName=neck_02}}
     {targetBone {boneName=neck_01}}
    ]
    constraintWeights [1,1,1,1,1,1]
   }
   n12 = "Constraint - dyn_ponytail_c_r (PPP)"!(ComponentPose=@n11, Alpha=1, ConstraintWeights_0=0.5381, ConstraintWeights_1:float=0.0681, ConstraintWeights_2:float=0.3938) -> (Pose) {
    alphaBoolBlend {blendOption=Linear}
    boneToModify {boneName=dyn_ponytail_c_r}
    constraintSetup [{targetBone {boneName=clavicle_r}}, {targetBone {boneName=spine_04}}, {targetBone {boneName=spine_05}}]
    constraintWeights [1,1,1]
   }
   n13 = "Constraint - dyn_ponytail_d_r (PPPPPP)"!(
    ComponentPose=@n12
    Alpha=1
    ConstraintWeights_0=0.0392
    ConstraintWeights_1:float=0.0366
    ConstraintWeights_2:float=0.3979
    ConstraintWeights_3:float=0.4536
    ConstraintWeights_4:float=0.0622
    ConstraintWeights_5:float=0.0106
   ) -> (Pose) {
    alphaBoolBlend {blendOption=Linear}
    boneToModify {boneName=dyn_ponytail_d_r}
    constraintSetup [
     {targetBone {boneName=spine_02}}
     {targetBone {boneName=spine_03}}
     {targetBone {boneName=spine_04}}
     {targetBone {boneName=spine_05}}
     {targetBone {boneName=clavicle_r}}
     {targetBone {boneName=pelvis}}
    ]
    constraintWeights [1,1,1,1,1,1]
   }
   n14 = "Constraint - dyn_ponytail_e_r (PPPPPP)"!(
    ComponentPose=@n13
    Alpha=1
    ConstraintWeights_0=0.0275
    ConstraintWeights_1:float=0.1032
    ConstraintWeights_2:float=0.0795
    ConstraintWeights_3:float=0.4475
    ConstraintWeights_4:float=0.2482
    ConstraintWeights_5:float=0.0942
   ) -> (Pose) {
    alphaBoolBlend {blendOption=Linear}
    boneToModify {boneName=dyn_ponytail_e_r}
    constraintSetup [
     {targetBone {boneName=spine_01}}
     {targetBone {boneName=spine_02}}
     {targetBone {boneName=spine_03}}
     {targetBone {boneName=spine_04}}
     {targetBone {boneName=spine_05}}
     {targetBone {boneName=pelvis}}
    ]
    constraintWeights [1,1,1,1,1,1]
   }
   n15 = "Constraint - dyn_ponytail_f_r (PPPPPP)"!(
    ComponentPose=@n14
    Alpha=1
    ConstraintWeights_0=0.0369
    ConstraintWeights_1:float=0.1212
    ConstraintWeights_2:float=0.081
    ConstraintWeights_3:float=0.3589
    ConstraintWeights_4:float=0.1346
    ConstraintWeights_5:float=0.2674
   ) -> (Pose) {
    alphaBoolBlend {blendOption=Linear}
    boneToModify {boneName=dyn_ponytail_f_r}
    constraintSetup [
     {targetBone {boneName=spine_01}}
     {targetBone {boneName=spine_02}}
     {targetBone {boneName=spine_03}}
     {targetBone {boneName=spine_04}}
     {targetBone {boneName=spine_05}}
     {targetBone {boneName=pelvis}}
    ]
    constraintWeights [1,1,1,1,1,1]
   }
   n16 = "Transform (Modify) Bone - Bone: dyn_ponytail_c_l"!(ComponentPose=@n15, bAlphaBoolEnabled:bool=True, Translation="0,1,-3", Rotation="0,0,0", Scale="1,1,1") -> (Pose) {
    alphaInputType=Bool
    rotationMode=BMM_Additive
    rotationSpace=BCS_BoneSpace
    translationMode=BMM_Additive
    translationSpace=BCS_BoneSpace
    alphaBoolBlend {blendOption=Linear}
    alphaScaleBiasClamp {inRange {max=40}}
    boneToModify {boneName=dyn_ponytail_c_l}
    translation {y=1, z=-3}
   }
   n17 = "Transform (Modify) Bone - Bone: dyn_ponytail_c_r"!(ComponentPose=@n16, bAlphaBoolEnabled:bool=True, Translation="0,-1,3", Rotation="0,0,0", Scale="1,1,1") -> (Pose) {
    alphaInputType=Bool
    rotationMode=BMM_Additive
    rotationSpace=BCS_BoneSpace
    translationMode=BMM_Additive
    translationSpace=BCS_BoneSpace
    alphaBoolBlend {blendOption=Linear}
    alphaScaleBiasClamp {inRange {max=40}}
    boneToModify {boneName=dyn_ponytail_c_r}
    translation {y=-1, z=3}
   }
   n18 = "Transform (Modify) Bone - Bone: dyn_ponytail_d_l"!(ComponentPose=@n17, bAlphaBoolEnabled:bool=True, Translation="0,0,0", Rotation="-5,10,0", Scale="1,1,1") -> (Pose) {
    alphaInputType=Bool
    rotationMode=BMM_Additive
    rotationSpace=BCS_BoneSpace
    translationMode=BMM_Additive
    translationSpace=BCS_BoneSpace
    alphaBoolBlend {blendOption=Linear}
    alphaScaleBiasClamp {inRange {max=40}}
    boneToModify {boneName=dyn_ponytail_d_l}
    rotation {pitch=-5, yaw=10}
   }
   n19 = "Transform (Modify) Bone - Bone: dyn_ponytail_d_r"!(ComponentPose=@n18, bAlphaBoolEnabled:bool=True, Translation="0,0,0", Rotation="-5,10,0", Scale="1,1,1") -> (Pose) {
    alphaInputType=Bool
    rotationMode=BMM_Additive
    rotationSpace=BCS_BoneSpace
    translationMode=BMM_Additive
    translationSpace=BCS_BoneSpace
    alphaBoolBlend {blendOption=Linear}
    alphaScaleBiasClamp {inRange {max=40}}
    boneToModify {boneName=dyn_ponytail_d_r}
    rotation {pitch=-5, yaw=10}
   }
   n20 = Anim Dynamics - Bone: dyn_hair_fr_a_l(ComponentPose=@n19, Alpha=1, ExternalForce="0,0,0") {
    angularDampingOverride=1
    angularSpringConstant=1500
    bAngularSpring=true
    bGravityOverrideInSimSpace=true
    bOverrideAngularDamping=true
    gravityScale=0.35
    simulationSpace=RootRelative
    alphaBoolBlend {blendOption=Linear}
    boundBone {boneName=dyn_hair_fr_a_l}
    physicsBodyDefinitions [{
      sphereCollisionRadius=0
      boundBone {boneName=dyn_hair_fr_a_l}
      boxExtents=<10 20 20>
      constraintSetup {angularLimitsMax=<0 0 20>, angularLimitsMin=<0 -25 0>, angularTarget=<0 0 0>}
      localJointOffset=<10 0 0>
     }]
   }
   n21 = Anim Dynamics - Bone: dyn_hair_fr_b_l(ComponentPose=@n20, Alpha=1, ExternalForce="0,0,0") {
    angularDampingOverride=1
    angularSpringConstant=1500
    bAngularSpring=true
    bGravityOverrideInSimSpace=true
    bOverrideAngularDamping=true
    gravityScale=0.35
    simulationSpace=RootRelative
    alphaBoolBlend {blendOption=Linear}
    boundBone {boneName=dyn_hair_fr_b_l}
    physicsBodyDefinitions [{
      sphereCollisionRadius=0
      boundBone {boneName=dyn_hair_fr_b_l}
      boxExtents=<10 20 20>
      constraintSetup {angularLimitsMax=<0 0 25>, angularLimitsMin=<0 -30 -15>, angularTarget=<0 0 0>}
      localJointOffset=<10 0 0>
     }]
   }
   Save cached pose 'NoDynamicPose'(Pose=Component To Local(ComponentPose=@n21)) {cacheName=NoDynamicPose}
   n22 = Get Body_Mesh()
   Save cached pose 'StartPose'(Pose=Layered blend per bone(BasePose=Copy Pose From Mesh() {bCopyCurves=true, bUseAttachedParent=true}, BlendPoses_0=Copy Pose From Mesh(SourceMeshComponent=@n22)) {layerSetup [{branchFilters [{boneName=upperarm_l}, {boneName=upperarm_r}]}]}) {cacheName=StartPose}
   n23 = Get headUp() -> (headUp:double)
   n24 = "Transform (Modify) Bone - Bone: dyn_ponytail_a_base_l"(
    ComponentPose=Local To Component(LocalPose=Use cached pose 'StartPose'())
    Alpha=@n23
    Translation="-1,-1,0"
    Rotation="0,-35,0"
    Scale="1,1,1"
   ) {
    rotationMode=BMM_Additive
    rotationSpace=BCS_BoneSpace
    translationMode=BMM_Additive
    translationSpace=BCS_BoneSpace
    alphaBoolBlend {blendOption=Linear}
    alphaScaleBiasClamp {bMapRange=true, inRange {max=0, min=-50}, outRange {max=0, min=1}}
    boneToModify {boneName=dyn_ponytail_a_base_l}
    rotation {yaw=-35}
    translation=<-1 -1>
   }
   n25 = "Transform (Modify) Bone - Bone: dyn_ponytail_a_base_r"(ComponentPose=@n24, Alpha=@n23, Translation="1,1,0", Rotation="0,-35,0", Scale="1,1,1") {
    rotationMode=BMM_Additive
    rotationSpace=BCS_BoneSpace
    translationMode=BMM_Additive
    translationSpace=BCS_BoneSpace
    alphaBoolBlend {blendOption=Linear}
    alphaScaleBiasClamp {bMapRange=true, inRange {max=0, min=-50}, outRange {max=0, min=1}}
    boneToModify {boneName=dyn_ponytail_a_base_r}
    rotation {yaw=-35}
    translation=<1 1>
   }
   n26 = Get headTwist() -> (headTwist:double)
   n27 = "Transform (Modify) Bone - Bone: dyn_ponytail_a_base_l"(ComponentPose=@n25, Alpha=@n26, Translation="0,0,0", Rotation="10,0,0", Scale="1,1,1") {
    rotationMode=BMM_Additive
    rotationSpace=BCS_BoneSpace
    translationMode=BMM_Additive
    translationSpace=BCS_BoneSpace
    alphaBoolBlend {blendOption=Linear}
    alphaScaleBiasClamp {bMapRange=true, inRange {max=40}}
    boneToModify {boneName=dyn_ponytail_a_base_l}
    rotation {pitch=10}
   }
   n28 = "Transform (Modify) Bone - Bone: dyn_ponytail_a_base_r"(ComponentPose=@n27, Alpha=@n26, Translation="0,0,2", Rotation="-30,0,0", Scale="1,1,1") {
    rotationMode=BMM_Additive
    rotationSpace=BCS_BoneSpace
    translationMode=BMM_Additive
    translationSpace=BCS_BoneSpace
    alphaBoolBlend {blendOption=Linear}
    alphaScaleBiasClamp {bMapRange=true, inRange {max=40}}
    boneToModify {boneName=dyn_ponytail_a_base_r}
    rotation {pitch=-30}
    translation {z=2}
   }
   n29 = Get headTwist()
   n30 = "Transform (Modify) Bone - Bone: dyn_ponytail_a_base_l"(ComponentPose=@n28, Alpha=@n29, Translation="0,0,-2", Rotation="-30,0,0", Scale="1,1,1") {
    rotationMode=BMM_Additive
    rotationSpace=BCS_BoneSpace
    translationMode=BMM_Additive
    translationSpace=BCS_BoneSpace
    alphaBoolBlend {blendOption=Linear}
    alphaScaleBiasClamp {bMapRange=true, inRange {max=0, min=-40}, outRange {max=0, min=1}}
    boneToModify {boneName=dyn_ponytail_a_base_l}
    rotation {pitch=-30}
    translation {z=-2}
   }
   n31 = "Transform (Modify) Bone - Bone: dyn_ponytail_a_base_r"(ComponentPose=@n30, Alpha=@n29, Translation="0,0,0", Rotation="10,0,0", Scale="1,1,1") {
    rotationMode=BMM_Additive
    rotationSpace=BCS_BoneSpace
    translationMode=BMM_Additive
    translationSpace=BCS_BoneSpace
    alphaBoolBlend {blendOption=Linear}
    alphaScaleBiasClamp {bMapRange=true, inRange {max=0, min=-40}, outRange {max=0, min=1}}
    boneToModify {boneName=dyn_ponytail_a_base_r}
    rotation {pitch=10}
   }
   n32 = Get jumpingAlpha() -> (jumpingAlpha:double)
   n33 = "Transform (Modify) Bone - Bone: dyn_ponytail_a_base_l"(ComponentPose=@n31, Alpha=@n32, Translation="0,0,0", Rotation="-5,-40,0", Scale="1,1,1") {
    rotationMode=BMM_Additive
    rotationSpace=BCS_BoneSpace
    translationMode=BMM_Additive
    translationSpace=BCS_BoneSpace
    alphaBoolBlend {blendInTime=50, blendOption=Linear, blendOutTime=50}
    alphaScaleBiasClamp {bInterpResult=true, inRange {max=0, min=-40}, outRange {max=0, min=1}}
    boneToModify {boneName=dyn_ponytail_a_base_l}
    rotation {pitch=-5, yaw=-40}
   }
   n34 = "Transform (Modify) Bone - Bone: dyn_ponytail_a_base_r"(ComponentPose=@n33, Alpha=@n32, Translation="0,0,0", Rotation="-5,-40,0", Scale="1,1,1") {
    rotationMode=BMM_Additive
    rotationSpace=BCS_BoneSpace
    translationMode=BMM_Additive
    translationSpace=BCS_BoneSpace
    alphaBoolBlend {blendInTime=50, blendOption=Linear, blendOutTime=50}
    alphaScaleBiasClamp {bInterpResult=true, inRange {max=0, min=-40}, outRange {max=0, min=1}}
    boneToModify {boneName=dyn_ponytail_a_base_r}
    rotation {pitch=-5, yaw=-40}
   }
   n35 = Get bIsSkyDiving()
   n36 = "Transform (Modify) Bone - Bone: dyn_ponytail_a_base_l"!(ComponentPose=@n34, bAlphaBoolEnabled:bool=@n35, Translation="0,0,0", Rotation="-10,-10,0", Scale="1,1,1") -> (Pose) {
    alphaInputType=Bool
    rotationMode=BMM_Additive
    rotationSpace=BCS_BoneSpace
    translationMode=BMM_Additive
    translationSpace=BCS_BoneSpace
    alphaBoolBlend {blendInTime=10, blendOption=Linear, blendOutTime=10}
    alphaScaleBiasClamp {inRange {max=0, min=-40}, outRange {max=0, min=1}}
    boneToModify {boneName=dyn_ponytail_a_base_l}
    rotation {pitch=-10, yaw=-10}
   }
   n37 = "Transform (Modify) Bone - Bone: dyn_ponytail_a_base_r"!(ComponentPose=@n36, bAlphaBoolEnabled:bool=@n35, Translation="0,0,0", Rotation="-10,-10,0", Scale="1,1,1") -> (Pose) {
    alphaInputType=Bool
    rotationMode=BMM_Additive
    rotationSpace=BCS_BoneSpace
    translationMode=BMM_Additive
    translationSpace=BCS_BoneSpace
    alphaBoolBlend {blendInTime=10, blendOption=Linear, blendOutTime=10}
    alphaScaleBiasClamp {inRange {max=0, min=-40}, outRange {max=0, min=1}}
    boneToModify {boneName=dyn_ponytail_a_base_r}
    rotation {pitch=-10, yaw=-10}
   }
   n38 = Anim Dynamics - Bone: dyn_hair_fr_a_l(ComponentPose=@n37, Alpha=1, ExternalForce="0,0,0") {
    angularDampingOverride=1
    angularSpringConstant=1500
    bAngularSpring=true
    bGravityOverrideInSimSpace=true
    bOverrideAngularDamping=true
    gravityScale=0.35
    simulationSpace=RootRelative
    alphaBoolBlend {blendOption=Linear}
    boundBone {boneName=dyn_hair_fr_a_l}
    physicsBodyDefinitions [{
      sphereCollisionRadius=0
      boundBone {boneName=dyn_hair_fr_a_l}
      boxExtents=<10 20 20>
      constraintSetup {angularLimitsMax=<0 0 20>, angularLimitsMin=<0 -25 0>, angularTarget=<0 0 0>}
      localJointOffset=<10 0 0>
     }]
   }
   n39 = Anim Dynamics - Bone: dyn_hair_fr_b_l(ComponentPose=@n38, Alpha=1, ExternalForce="0,0,0") {
    angularDampingOverride=1
    angularSpringConstant=1500
    bAngularSpring=true
    bGravityOverrideInSimSpace=true
    bOverrideAngularDamping=true
    gravityScale=0.35
    simulationSpace=RootRelative
    alphaBoolBlend {blendOption=Linear}
    boundBone {boneName=dyn_hair_fr_b_l}
    physicsBodyDefinitions [{
      sphereCollisionRadius=0
      boundBone {boneName=dyn_hair_fr_b_l}
      boxExtents=<10 20 20>
      constraintSetup {angularLimitsMax=<0 0 25>, angularLimitsMin=<0 -30 -15>, angularTarget=<0 0 0>}
      localJointOffset=<10 0 0>
     }]
   }
   n40 = RigidBody!(
    ComponentPose=@n39
    Alpha
    OverrideWorldGravity:Vector=Get GravityOverride()
    ExternalForce=Get externalForce() -> (externalForce:Vector)
    ComponentLinearAccScale:Vector="0.4,0.35,0.35"
    ComponentLinearVelScale:Vector="0.4,0.35,0.35"
   ) -> (Pose) {
    bClampLinearTranslationLimitToRefPose=true
    bOverrideWorldGravity=true
    overridePhysicsAsset=&/F_MED_Sailor_Squad_Rebel_FaceAcc_Physics
    simulationSpace=BaseBoneSpace
    alphaBoolBlend {blendOption=Linear}
    baseBoneRef {boneName=neck_01}
    componentLinearAccScale=<0.6 0.35 0.35>
    componentLinearVelScale=<0.6 0.35 0.35>
    overrideWorldGravity {z=-980}
   }
   n41 = Get bIsSkydiveFloating()
   n42 = "Transform (Modify) Bone - Bone: dyn_ponytail_c_l"(ComponentPose=@n40, bAlphaBoolEnabled=@n41, Translation="0,0,0", Rotation="0,-15,0", Scale="1,1,1") {
    alphaInputType=Bool
    rotationMode=BMM_Additive
    rotationSpace=BCS_BoneSpace
    translationMode=BMM_Additive
    translationSpace=BCS_BoneSpace
    alphaBoolBlend {blendInTime=5, blendOption=Linear, blendOutTime=5}
    alphaScaleBiasClamp {inRange {max=0, min=-40}, outRange {max=0, min=1}}
    boneToModify {boneName=dyn_ponytail_c_l}
    rotation {yaw=-15}
   }
   n43 = "Transform (Modify) Bone - Bone: dyn_ponytail_c_r"(ComponentPose=@n42, bAlphaBoolEnabled=@n41, Translation="0,0,0", Rotation="0,-15,0", Scale="1,1,1") {
    alphaInputType=Bool
    rotationMode=BMM_Additive
    rotationSpace=BCS_BoneSpace
    translationMode=BMM_Additive
    translationSpace=BCS_BoneSpace
    alphaBoolBlend {blendInTime=5, blendOption=Linear, blendOutTime=5}
    alphaScaleBiasClamp {inRange {max=0, min=-40}, outRange {max=0, min=1}}
    boneToModify {boneName=dyn_ponytail_c_r}
    rotation {yaw=-15}
   }
   n44 = "Transform (Modify) Bone - Bone: dyn_ponytail_d_l"(ComponentPose=@n43, bAlphaBoolEnabled=@n41, Translation="0,0,0", Rotation="0,-40,0", Scale="1,1,1") {
    alphaInputType=Bool
    rotationMode=BMM_Additive
    rotationSpace=BCS_BoneSpace
    translationMode=BMM_Additive
    translationSpace=BCS_BoneSpace
    alphaBoolBlend {blendInTime=5, blendOption=Linear, blendOutTime=5}
    alphaScaleBiasClamp {inRange {max=0, min=-40}, outRange {max=0, min=1}}
    boneToModify {boneName=dyn_ponytail_d_l}
    rotation {yaw=-40}
   }
   n45 = "Transform (Modify) Bone - Bone: dyn_ponytail_d_r"(ComponentPose=@n44, bAlphaBoolEnabled=@n41, Translation="0,0,0", Rotation="0,-40,0", Scale="1,1,1") {
    alphaInputType=Bool
    rotationMode=BMM_Additive
    rotationSpace=BCS_BoneSpace
    translationMode=BMM_Additive
    translationSpace=BCS_BoneSpace
    alphaBoolBlend {blendInTime=5, blendOption=Linear, blendOutTime=5}
    alphaScaleBiasClamp {inRange {max=0, min=-40}, outRange {max=0, min=1}}
    boneToModify {boneName=dyn_ponytail_d_r}
    rotation {yaw=-40}
   }
   Save cached pose 'DynamicPose'(Pose=Component To Local(ComponentPose=@n45)) {cacheName=DynamicPose}
  ]}
 EventGraph {nodes [
   Event OnCharacterCustomizationLoaded()
   > Sequence()
   n1 = CheckForBackbling()
   > n2 = Set Is Wearing BP(Is Wearing BP:bool=OR Boolean!(A=@n1, B=@n1.bIsBackblingACape, C:bool=@n1.bIsBackblingLong) -> (ReturnValue)) -> (Output_Get:bool)
   > n3 = Set Is Wearing Cape(Is Wearing Cape=@n1.bIsBackblingLong)
   > Set Is Backbling Equipped(Is Backbling Equipped:bool=OR Boolean!(A=@n3, B=@n2) -> (ReturnValue)) -> (Output_Get:bool)
   Event BlueprintUpdateAnimation()
   > n4 = Sequence!(execute) -> (then_0, then_1, then_2:Exec, then_3:Exec, then_4:Exec, then_5:Exec, then_6:Exec)
   n5 = Branch(execute=@n4.then_4, Condition=Get bIsJumping())
   > Set externalForce(externalForce:Vector="0,-2000,0") -> (Output_Get:Vector)
   Set externalForce(execute=@n5.else, externalForce="0, 0, 0")
   n6 = Gravity Override State Function DEPRECATED(
    execute=@n4.then_5
    CharacterPartAnimInstance=Self-Reference()
    GameTimeInSeconds=GetGameTimeInSeconds()
    GravityOverrideParams=Get Gravity Override Params()
   )
   > Set GravityOverride(GravityOverride=@n6)
   n7 = Branch(execute=@n4.then_3, Condition=OR Boolean(A=Get bIsFalling(), B=Get bIsJumping()))
   > Set jumpingAlpha(jumpingAlpha:double=1) -> (Output_Get:double)
   Set jumpingAlpha(execute=@n7.else, jumpingAlpha=0)
   n8 = GetOwnerSkeletalMeshComponent()
   n9 = BreakRotator(InRot=BreakTransform(InTransform=GetDeltaTransformFromRefPose(self=@n8, BoneName=head, BaseName=spine_05)).Rotation)
   Set headUp(execute=@n4.then_0, headUp:double=@n9.Yaw) -> (Output_Get:double)
   > Set headTwist(headTwist:double=@n9) -> (Output_Get:double)
   n10 = Get Pawn Owner() -> (Pawn Owner:Player Pawn Athena Object Reference)
   Is Valid(exec=@n4.then_2, InputObject=@n10)
   > Do Once()
   > n11 = Set IsBackpackOn(IsBackpackOn:bool="Not Equal (Object)"!(A:Object Reference=GetSkeletalMeshForPartType(self=@n10, PartType=Backpack), B:Object Reference) -> (ReturnValue:bool)) -> (Output_Get:bool)
   > PrintString!(execute, InString="To String (Boolean)"(InBool=@n11), bPrintToScreen, bPrintToLog, TextColor, Duration, Key) -> (then)
   n12 = Get bRigidBodyEnabled()
   n13 = Get CurrentLOD()
   Set IsRigidBodyDisabled(
    execute=@n4.then_1
    IsRigidBodyDisabled:bool=OR Boolean!(
     A=NOT Boolean(A=@n12)
     B="integer >= integer"!(A:int=@n13, B:int=1) -> (ReturnValue:bool) {is=title K2Node_CallFunction}
     C:bool=Get Is Backbling Equipped() -> (Is Backbling Equipped:bool)
    ) -> (ReturnValue)
   ) -> (Output_Get:bool)
   Event BlueprintInitializeAnimation()
   > Set Body_Mesh(Body_Mesh=GetOwnerSkeletalMeshComponent())
   > n14 = Cast To PlayerPawn_Athena(Object=TryGetPawnOwner())
   > n15 = Set Pawn Owner(Pawn Owner:Player Pawn Athena Object Reference=@n14) -> (Output_Get:Player Pawn Athena Object Reference)
   > Set Head_Mesh(Head_Mesh:Skeletal Mesh Component Object Reference=GetSkeletalMeshForPartType(self=@n15, PartType=Head)) -> (Output_Get:Skeletal Mesh Component Object Reference)
   TryGetPawnOwner()
   Get IsBackpackOn() -> (IsBackpackOn:bool)
  ]}
]
variables [
 Body_Mesh {type=SkeletalMeshComponent}
 externalForce {type=Vector}
 Gravity Override Params {type=GravityOverrideParamsStruct, default {skydive_FloatingDown {gravityOverride=<0 -300 100>}}}
 GravityOverride {type=Vector}
 Head_Mesh {type=SkeletalMeshComponent}
 headTwist {type=number}
 headUp {type=number}
 Is Backbling Equipped {type=boolean}
 Is Wearing BP {type=boolean}
 Is Wearing Cape {type=boolean}
 IsBackpackOn {type=boolean}
 IsRigidBodyDisabled {type=boolean}
 jumpingAlpha {type=number}
 Pawn Owner {type=PlayerPawn_Athena_C}
 Skel_Mesh {type=SkeletalMeshComponent}
]
